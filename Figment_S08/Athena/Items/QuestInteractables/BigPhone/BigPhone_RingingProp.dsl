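; Quest prop for the Big Phone interactable that rings when a player dials the phone sharing its pairing int.
; At match ready on authority, it binds to matching phones' dial delegates.
; A dial plays audio immediately and via two delayed timers, flushing net dormancy to replicate.
; A destroyed-actor handler stops audio and blocks further rings, though its binding is not shown here; ShouldDie always returns false.
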(event EventBeginPlay
  ;[1 1 1] Wait until assets are loaded on the server, get our phone actors, and bind their dial delegates if we match their own pairing int value
  (Class|FortAsyncActionAuthorityMatchReady|MatchReadyAsync
    (:OnReady
      (bind _outactors (Actor|GetAllActorsOfClass "/Figment_S08/Athena/Items/QuestInteractables/BigPhone/Prop_Quest_BigPhone_Parent.Prop_Quest_BigPhone_Parent_C"))
      (for _array_element _outactors
        (if (== (Variables|Default|GetPairingInt) (Variables|/Figment_S08/Athena/Items/QuestInteractables/BigPhone/Prop_Quest_BigPhone_Parent.Prop_Quest_BigPhone_Parent_C|GetCallPairingInt _array_element))
          (Default|BindEventtoOnPlayerDialed _array_element (delegate Custom|OnPlayerRangMe)))))))

(event Custom|OnPlayerRangMe
  ;[1 1 1] If paired actor is alive, play your audio and replicate and then go back to being dormant
  (CallFunction|PlayAudio)
  (Utilities|Time|SetTimerbyEvent (delegate Custom|PlayAudio) (Variables|Default|GetFirstRingDelay))
  (Utilities|Time|SetTimerbyEvent (delegate Custom|PlayAudio2) (Variables|Default|GetSecondRingDelay)))

(event Custom|PlayAudio
  (Development|PrintString "RingRing I got your delegate")
  (if (not (Variables|Default|GetbIsPairedActorDead))
    (Audio|Components|Audio|Play (Variables|Default|GetAudio) 0.0)
    (Networking|FlushNetDormancy)))

(event Custom|CustomEvent_1
  (params (DestroyedActor :type Actor%20Object%20Reference))
  ;[1 1 1] If the attached actor is destroyed, stop any ringinging and prevent future rings (Maybe kill ourselves as well?)
  (Audio|Components|Audio|Stop (Variables|Default|GetAudio))
  (Variables|Default|SetbIsPairedActorDead true)
  (Networking|FlushNetDormancy))

(event Custom|PlayAudio2
  (Development|PrintString "RingRing I got your delegate")
  (if (not (Variables|Default|GetbIsPairedActorDead))
    (Audio|Components|Audio|Play (Variables|Default|GetAudio) 0.0)
    (Networking|FlushNetDormancy)))

(fn ShouldDie
  (params (Damage :type Float%20%28single-precision%29) (EventInstigator :type Controller%20Object%20Reference) (DamageCauser :type Actor%20Object%20Reference) (bIsTest :type Boolean) (EffectContext :type Gameplay%20Effect%20Context%20Handle%20Structure%20%28by%20ref%29))
  (return false))

(fn ConstructionScript)
